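# Translucent unlit hologram material for Niagara mesh particles in the ClaimReflect drum kit cosmetic effects.
# Emissive blends a two-color tint with ClaimReflect backpack textures and an optional pulsing red flash, shaped by stabilized screen-space scanlines, Fresnel and normal-mapped fake lighting.
# Desaturated scene color adapts brightness to dark or light backgrounds (constant on mobile), then a sphere mask and particle color modulate it.
# Separate nodes compute opacity terms and time-driven Z-axis rotation.
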Material /SparksCosmetics/Instrument/DrumKit/ClaimReflect/Effects/M_ClaimReflect_Hologram
graph {
 properties {
  assetImportData="<self>:InterchangeAssetImportData_0"
  bUsedWithNiagaraMeshParticles=true
  blendMode=BLEND_Translucent
  previewMesh=&/SM_ClaimReflect_Ring_Broken
  shadingModel=MSM_Unlit
  usedShadingModels=Unlit
 }
 nodes [
  Constant() {r=1}
  n1 = ParticleColor()
  n2 = SceneColor!(UVs)
  n3 = Constant() {r=1}
  n4 = ShadingPathSwitch(
   Default=Clamp(None=Desaturation(None=@n2.RGB, Fraction=@n3), Max=ScalarParameter() {defaultValue=0.9, parameterName=Max Opaqueness})
   Mobile=Constant() {r=1}
  ) {desc=Migrated from FeatureLevelSwitch}
  n5 = LinearInterpolate(
   A=VectorParameter() {parameterName=Color A, defaultValue=<rgba 0 0.4701 1.2118 1>}.RGB
   B=VectorParameter() {parameterName=Color B, defaultValue=<rgba 0 1.4123 2 1>}.RGB
   Alpha=NamedRerouteUsage()
  )
  n6 = TextureSample() {texture=/BRCosmetics/Accessories/FORT_Backpacks/Backpack_F_MED_ClaimReflect/Textures/T_ClaimReflect_Backpack_D}
  n7 = Constant() {r=-1}
  n8 = Add(
   A=LinearInterpolate(B=Desaturation(None=@n6.RGB, Fraction=@n7), Alpha=ScalarParameter() {defaultValue=3, parameterName=Diffuse Color Contribution}) {constAlpha=2}
   B=Multiply(A=TextureSample() {texture=/BRCosmetics/Accessories/FORT_Backpacks/Backpack_F_MED_ClaimReflect/Textures/T_ClaimReflect_Backpack_E}.RGB) {constB=3}
  )
  n9 = Add(A=@n8, B=@n5)
  n10 = LinearInterpolate(
   A=Constant3Vector() {constant=<rgba 1 0 0 0>}
   B=Add(A=@n5, B=@n8)
   Alpha=MaterialFunctionCall(Speed=Constant() {r=1}) {materialFunction=&DebugTimeSine}.Result
  )
  n11 = ScalarParameter() {parameterName=FlashRedColor}
  n12 = Fresnel() {baseReflectFraction=0.3, exponent=6}
  n13 = MaterialFunctionCall!(
   ScanLine_Tiling_1=Constant()
   Scanline_PannerSpeed_1=Constant() {r=0.3}
   ScanLine_Tiling=Constant() {r=130}
   Scanline_PannerSpeed=Constant() {r=0.5}
   UVs=MaterialFunctionCall!(Position, UVRadialScale=Constant() {r=0.1}) {materialFunction=/Game/Weapons/WeaponSkins/MaterialFunction/ObjectStabilizedScreenSpaceUVs}.StabilizedUVs
  ) {materialFunction=/Game/Weapons/WeaponSkins/MaterialFunction/ScanLines}
  n14 = ScalarParameter() {defaultValue=0.85, parameterName=ScanlineWeight, sliderMax=1}
  n15 = TextureSample() {
   samplerType=SAMPLERTYPE_Normal
   texture=/BRCosmetics/Accessories/FORT_Backpacks/Backpack_F_MED_ClaimReflect/Textures/T_ClaimReflect_Backpack_N
  }
  n16 = Constant() {r=0.1}
  n17 = Add(
   A=Multiply(
    A=@n15.RGB
    B=MaterialFunctionCall(In=@n15.RGB, Contrast=@n16) {materialFunction=&CheapContrast}.Result
   )
   B=@n15.RGB
  )
  n18 = OneMinus(None=Fresnel(Normal=@n17))
  n19 = Transform(None=Normalize(VectorInput=VectorParameter() {defaultValue=<rgba 0.576 0.2472 0.4977 0>}.RGB)) {transformSourceType=TRANSFORMSOURCE_View}
  n20 = LinearInterpolate(
   A=Fresnel(Normal=VertexNormalWS())
   B=ConstantBiasScale(None=DotProduct(A=@n17, B=@n19)) {scale=0.468}
   Alpha=ScalarParameter() {defaultValue=0.35, parameterName=Lighting Mix, sliderMax=1}
  )
  n21 = Power(
   Base=Multiply(A=Add(A=@n18, B=@n20), B=ScalarParameter() {defaultValue=0.9, parameterName=Lighting Weight, sliderMax=1})
   Exp=ScalarParameter() {defaultValue=1.44, parameterName=Lighting Mix Exponent, sliderMax=12}
  ) {constExponent=3}
  n22 = Add(A=Multiply(A=Multiply(A=@n13.Result, B=@n14), B=OneMinus(None=@n21)), B=@n21)
  n23 = Multiply(
   A=LinearInterpolate(
    A=ScalarParameter() {defaultValue=1, parameterName="Brightness (Dark BG)"}
    B=ScalarParameter() {defaultValue=5, parameterName="Brightness (Light BG)"}
    Alpha=@n4
   )
   B=Multiply(A=LinearInterpolate(A=@n9, B=@n10, Alpha=@n11), B=LinearInterpolate(A=@n12, Alpha=@n22) {constA=0.3})
  )
  n24 = ObjectPositionWS()
  n25 = WorldPosition()
  n26 = ObjectRadius()
  n27 = ScalarParameter() {defaultValue=0.1, parameterName=Hardness, sliderMax=1}
  n28 = LinearInterpolate(
   A=ScalarParameter() {defaultValue=0.8, parameterName="Opacity Multiplier (Edge)"}
   B=ScalarParameter() {defaultValue=1, parameterName="Opacity Multiplier (Center)"}
   Alpha=NamedRerouteDeclaration(None=SphereMask(A=@n24, B=@n25.XYZ, Radius=@n26, Hardness=@n27) {hardnessPercent=0.1}) {name=Sphere Mask, nodeColor=<rgba 1 0.7969 0 1>}
  ) {constA=0.2, constB=0.95}
  MaterialFunctionCall(Emissive=Multiply(A=@n1.RGB, B=Multiply(A=@n23, B=@n28))) {materialFunction=&MF_FX_EmissiveAdjustments}
  MaterialFunctionCall!(Center, Height) {materialFunction=/Game/Effects/Fort_Effects/Materials/_Functions/MF_Z_GradientFromObjectHeight_01}
  n29 = DotProduct(A=CameraVectorWS(), B=VertexNormalWS())
  Multiply(A=ParticleColor().A, B=@n29)
  Multiply(A=MaterialFunctionCall(Exponent=Constant() {r=5}) {materialFunction=&MF_Fresnel_FX}.Fresnel)
  Multiply(
   A=LinearInterpolate(A=ScalarParameter() {parameterName="Opacity (Dark BG)"}, B=ScalarParameter() {defaultValue=1, parameterName="Opacity (Light BG)"}, Alpha=@n4) {constA=1.5}
   B=@n1.A
  )
  ObjectPositionWS()
  ParticlePositionWS()
  ParticlePositionWS()
  n30 = Constant3Vector() {constant=<rgba 0 0 1 0>}
  n31 = Frac(None=Multiply(A=DynamicParameter() {defaultValue=<rgba 0.15 0.075 1 1>, paramNames [RotationSpeed, Scanline Lerp, Param3, Param4]}.RotationSpeed, B=Time()))
  Reroute(None=MaterialFunctionCall(Rotation Axis=@n30, Rotation Angle=@n31) {materialFunction=&FixRotateAboutAxisNormals}.New World Space Vertex Normal)
  Reroute(None=RotateAboutAxis(NormalizedRotationAxis=@n30, RotationAngle=@n31, PivotPoint=ActorPositionWS(), Position=WorldPosition().XYZ))
  n32 = DynamicParameter() {defaultValue=<rgba 0.15 0.075 1 1>, paramNames [RotationSpeed, Scanline Lerp, Param3, Param4]}
  Saturate(None=Multiply(B=LinearInterpolate(A=@n1.A, B=@n13.Result, Alpha=@n32.Scanline Lerp) {constAlpha=0.1}) {constA=0.2})
  Step(Y=@n29) {constX=0.5}
  TextureCoordinate()
  n33 = VertexInterpolator(VS=ComponentMask(None=PreSkinnedPosition()) {b=true, g=true})
  n34 = Constant()
  n35 = ScalarParameter() {defaultValue=-0.5, parameterName=Scanline Pan Speed, sliderMax=5, sliderMin=-5}
  TextureSampleParameter2D(UVs=Panner(Coordinate=Divide(A=@n33.PS) {constB=30}, Time=Time(), Speed=AppendVector(A=@n34, B=@n35)) {speedY=-1}) {parameterName=Param, texture=/Game/PvP/StrategicBuildingActors/Defensive/LaserFence/Assets/LaserScanline}
 ]
}
