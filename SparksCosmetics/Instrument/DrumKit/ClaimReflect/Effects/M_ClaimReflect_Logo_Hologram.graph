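# Translucent, unlit, two-sided hologram material for Niagara mesh particles, showing the UNSC logo for the ClaimReflect drum kit.
# A two-color blend is tinted by particle color and multiplied by panning scanlines, with a sphere mask declared for gradients.
# The logo texture combines with scanlines and particle alpha for opacity.
# Further branches use Fresnel and desaturated scene color to adapt brightness and opacity to the background, with a constant fallback on ES3.1.
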Material /SparksCosmetics/Instrument/DrumKit/ClaimReflect/Effects/M_ClaimReflect_Logo_Hologram
graph {
 properties {
  assetImportData="<self>:InterchangeAssetImportData_0"
  bUsedWithNiagaraMeshParticles=true
  blendMode=BLEND_Translucent
  opacityMaskClipValue=0.333
  shadingModel=MSM_Unlit
  translucencyLightingMode=TLM_Surface
  twoSided=true
  usedShadingModels=Unlit
 }
 nodes [
  n1 = VectorParameter() {parameterName=Color A, defaultValue=<rgba 0 0.4701 1.2118 1>}
  n2 = VectorParameter() {parameterName=Color B, defaultValue=<rgba 0.4757 1.1103 2 1>}
  n3 = NamedRerouteUsage()
  n4 = ParticleColor()
  n5 = LinearInterpolate(
   B=MaterialFunctionCall!(
    ScanLine_Tiling_1=Constant() {r=10}
    Scanline_PannerSpeed_1=Constant() {r=0.3}
    ScanLine_Tiling=Constant() {r=70}
    Scanline_PannerSpeed=Constant() {r=0.5}
    UVs=TextureCoordinate()
   ) {materialFunction=/Game/Weapons/WeaponSkins/MaterialFunction/ScanLines}.Result
   Alpha=ScalarParameter() {defaultValue=0.5, parameterName=ScanlineWeight, sliderMax=1}
  ) {constA=1}
  n6 = Multiply(A=Multiply(A=LinearInterpolate(A=@n1.RGB, B=@n2.RGB, Alpha=@n3), B=@n4.RGB), B=@n5)
  MaterialFunctionCall(Emissive=@n6) {materialFunction=&MF_FX_EmissiveAdjustments}
  MaterialFunctionCall!(Position, UVRadialScale=Constant() {r=0.1}) {materialFunction=/Game/Weapons/WeaponSkins/MaterialFunction/ObjectStabilizedScreenSpaceUVs}
  n7 = SceneColor!(UVs)
  n8 = Constant() {r=1}
  n9 = FeatureLevelSwitch(
   Default=Clamp(None=Desaturation(None=@n7.RGB, Fraction=@n8), Max=ScalarParameter() {defaultValue=0.9, parameterName=Max Opaqueness})
   ES3_1=Constant() {r=1}
  )
  Multiply(
   A=LinearInterpolate(
    A=ScalarParameter() {defaultValue=1, parameterName="Brightness (Dark BG)"}
    B=ScalarParameter() {defaultValue=5, parameterName="Brightness (Light BG)"}
    Alpha=@n9
   )
   B=@n6
  )
  n10 = Fresnel(Normal=Abs(None=VertexNormalWS()))
  n11 = ScalarParameter() {parameterName="Opacity (Dark BG)"}
  n12 = ScalarParameter() {defaultValue=1, parameterName="Opacity (Light BG)"}
  n13 = TextureSampleParameter2D() {
   parameterName=Opacity Mask Texture
   samplerType=SAMPLERTYPE_LinearColor
   texture=&/T_ClaimReflect_DrumKit_UNSC_Logo
  }
  n14 = Multiply(A=@n13.R, B=@n5)
  Multiply(A=@n10, B=Multiply(A=LinearInterpolate(A=@n11, B=@n12, Alpha=@n9) {constA=1.5}, B=@n14))
  NamedRerouteDeclaration(None=SphereMask(
    A=ObjectPositionWS()
    B=WorldPosition().XYZ
    Radius=ObjectRadius()
    Hardness=ScalarParameter() {defaultValue=0.1, parameterName=Hardness, sliderMax=1}
   ) {hardnessPercent=0.1}) {name=Sphere Mask, nodeColor=<rgba 1 0.7969 0 1>}
  OneMinus(None=@n10)
  Saturate(None=Multiply(A=@n4.A, B=@n14))
 ]
}
