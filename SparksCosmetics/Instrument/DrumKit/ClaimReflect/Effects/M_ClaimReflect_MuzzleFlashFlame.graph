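# Additive, unlit, two-sided muzzle flash flame material for the ClaimReflect drum kit effects, flagged for Niagara sprites and meshes.
# A distorted flame-shape mask is modulated by panning noise, with an Is Radial Core scalar blending linear and radial UVs.
# The result indexes a color curve atlas, tinted by particle color, hue-shifted and remapped to HDR.
# Dynamic parameters drive noise amount, erosion and core selection; Fresnel-style, depth and UV-edge fades soften opacity.
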Material /SparksCosmetics/Instrument/DrumKit/ClaimReflect/Effects/M_ClaimReflect_MuzzleFlashFlame
graph {
 properties {
  bUsedWithMeshDeformer=true
  bUsedWithNiagaraMeshParticles=true
  bUsedWithNiagaraSprites=true
  bUsedWithSkeletalMesh=true
  blendMode=BLEND_Additive
  pixelDepthOffsetMode=PDOM_Legacy
  shadingModel=MSM_Unlit
  twoSided=true
  usedShadingModels=Unlit
 }
 nodes [
  Constant() {r=1}
  n1 = TextureCoordinate()
  n2 = AppendVector(
   A=ScalarParameter() {defaultValue=0.25, group=Noise, parameterName=U Distort Freq, sortPriority=5}
   B=ScalarParameter() {defaultValue=0.25, group=Noise, parameterName=V Distort Freq, sortPriority=6}
  )
  n3 = Multiply(A=TextureCoordinate(), B=@n2)
  n4 = Subtract(A=@n3, B=Multiply(A=@n2) {constB=0.5})
  n5 = LinearInterpolate(
   A=Panner(Coordinate=@n3) {speedX=0.01, speedY=0.05}
   B=Panner(Coordinate=MaterialFunctionCall(Vector or UVs=@n4) {materialFunction=&VectorToRadialValue}.Radial Coordinates) {speedX=0.01, speedY=-0.05}
   Alpha=ScalarParameter() {parameterName=Is Radial Core}
  ) {constAlpha=0}
  n6 = Multiply(
   A=ComponentMask(None=TextureSample(UVs=@n5) {samplerType=SAMPLERTYPE_Normal, texture=/Game/Effects/Fort_Effects/Textures/Tile/noise_flamy_Tile_NRM}.RGB) {g=true, r=true}
   B=ScalarParameter() {defaultValue=0.15, group=Noise, parameterName=Distortion Amplitude, sortPriority=7}
  ) {constB=0.15}
  n7 = TextureSampleParameter2D(UVs=Clamp(None=Add(A=@n1, B=@n6))) {group=FlameShape, parameterName=MuzzleShape, texture=/Game/Effects/Niagara/Systems/Weapon_Masters/T_MuzzleFlashes_M}
  n8 = LinearInterpolate(
   A=LinearInterpolate(A=@n7.R, B=@n7.R, Alpha=VertexColor().R)
   B=LinearInterpolate(
    A=@n7.B
    B=@n7.B
    Alpha=DynamicParameter() {defaultValue=<rgba 0.1 1 1 1>, paramNames [Noise Amount, Erosion, Core Selection, Param4]}.Core Selection
   )
   Alpha=ScalarParameter() {parameterName=Is Radial Core}
  ) {constAlpha=0}
  n9 = ScalarParameter() {defaultValue=0.25, group=Noise, parameterName=U Distort Freq, sortPriority=5}
  n10 = ScalarParameter() {defaultValue=0.25, group=Noise, parameterName=V Distort Freq, sortPriority=6}
  n11 = Panner(Coordinate=Multiply(A=TextureCoordinate(), B=AppendVector(A=@n9, B=@n10))) {speedY=1}
  n12 = Multiply(
   A=ComponentMask(None=TextureSampleParameter2D(UVs=@n11) {group=Noise, parameterName=Texture, sortPriority=4, texture=/Game/Weapons/FORT_Pistols/Effects/Muzzle/T_Cloud_d}.RGB) {g=true, r=true}
   B=ScalarParameter() {defaultValue=0.15, group=Noise, parameterName=Distortion Amplitude, sortPriority=7}
  )
  n13 = TextureCoordinate()
  n14 = ComponentMask(None=@n13) {r=true}
  n15 = ScalarParameter() {defaultValue=1, parameterName=Noise_U_Scale}
  n16 = ComponentMask(None=@n13) {g=true}
  n17 = ScalarParameter() {defaultValue=1, parameterName=Noise_V_Scale}
  n18 = AppendVector(
   A=Add(A=Multiply(A=@n14, B=@n15), B=ScalarParameter() {parameterName=Nouse_U_Offset})
   B=Add(A=Multiply(A=@n16, B=@n17), B=ScalarParameter() {parameterName=Nouse_V_Offset})
  )
  n19 = AppendVector(
   A=ScalarParameter() {defaultValue=0.25, group=Noise, parameterName=U Distort Freq, sortPriority=5}
   B=ScalarParameter() {defaultValue=0.25, group=Noise, parameterName=V Distort Freq, sortPriority=6}
  )
  n20 = Subtract(A=Multiply(A=TextureCoordinate(), B=@n19), B=Multiply(A=@n19) {constB=0.5})
  n21 = TextureSampleParameter2D(UVs=Panner(Coordinate=MaterialFunctionCall(Vector or UVs=@n20) {materialFunction=&VectorToRadialValue}.Radial Coordinates) {speedY=-1}) {group=Noise, parameterName=Texture, sortPriority=4, texture=&T_Cloud_d}
  n22 = TextureCoordinate()
  n23 = Multiply(A=ComponentMask(None=@n22) {r=true}, B=ScalarParameter() {defaultValue=1, parameterName=Noise_U_Scale})
  n24 = Multiply(A=ComponentMask(None=@n22) {g=true}, B=ScalarParameter() {defaultValue=1, parameterName=Noise_V_Scale})
  n25 = AppendVector(A=Add(A=@n23, B=ScalarParameter() {parameterName=Nouse_U_Offset}), B=Add(A=@n24, B=ScalarParameter() {parameterName=Nouse_V_Offset}))
  n26 = Multiply(A=AppendVector(A=@n23, B=@n24)) {constB=0.5}
  n27 = Add(
   A=Multiply(
    A=ComponentMask(None=@n21.RGB) {g=true, r=true}
    B=ScalarParameter() {defaultValue=0.15, group=Noise, parameterName=Distortion Amplitude, sortPriority=7}
   )
   B=MaterialFunctionCall(Vector or UVs=Subtract(A=@n25, B=@n26)) {materialFunction=&VectorToRadialValue}.Radial Coordinates
  )
  n28 = TextureSampleParameter2D(UVs=LinearInterpolate(
    A=Add(A=@n12, B=@n18)
    B=MaterialFunctionCall(Vector or UVs=@n27) {materialFunction=&VectorToRadialValue}.Radial Coordinates
    Alpha=ScalarParameter() {parameterName=Is Radial Core}
   ) {constAlpha=0}) {
   group=Noise
   parameterName=Noise
   samplerType=SAMPLERTYPE_LinearGrayscale
   texture=/Game/Effects/Fort_Effects/Emotes/UltraLight/Texture/T_FluidFlameTile_Greyscale
  }
  n29 = Power(Base=@n28.R, Exp=ScalarParameter() {defaultValue=1, group=Noise, parameterName=Noise Exp}) {constExponent=1.5}
  n30 = DynamicParameter() {defaultValue=<rgba 0.1 1 1 1>, paramNames [Noise Amount, Erosion, Core Selection, Param4]}
  n31 = LinearInterpolate(A=@n8, B=Multiply(A=@n8, B=Saturate(None=@n29)), Alpha=@n30.Noise Amount)
  n32 = CurveAtlasRowParameter(CurveTime=@n31) {
   atlas=&/CurveAtlas_ClaimReflectTurret
   curve=&/Curve_ClaimReflect_Turret_MuzzleFire
   parameterName=Curves
  }
  n33 = ParticleColor()
  n34 = Multiply(
   A=MaterialFunctionCall(Hue Shift Percentage=ScalarParameter() {parameterName=HueShift}, Texture=Multiply(A=@n32, B=@n33.RGB)) {materialFunction=&HueShift}.Result
   B=@n33.RGB
  )
  n35 = ScalarParameter() {defaultValue=1, group=FlameShape, parameterName=Max_HDR_Intensity}
  n36 = DotProduct(A=CameraVectorWS(), B=VertexNormalWS())
  n37 = SmoothStep(Value=Power(
    Base=Abs(None=@n36) {desc=Make this work on two-sided materials}
    Exp=ScalarParameter() {defaultValue=0.5, parameterName="FadeSensitivity (optional) (S)"}
   ))
  n38 = LinearInterpolate(A=ScalarParameter() {defaultValue=1, group=FlameShape, parameterName=Min_HDR_Intensity}, B=@n35, Alpha=@n37)
  MaterialFunctionCall(Emissive=MaterialFunctionCall(
    LDR Input=Clamp(None=@n34)
    HDR Tint=@n34
    HDR Intensity Multiplier=StaticSwitchParameter(True=@n38, False=@n35) {group=FlameShape, parameterName=Use Complex HDR Multiplier}
   ) {materialFunction=&DeriveHDRfromLDR}.RemappedRange) {materialFunction=&MF_FX_EmissiveAdjustments}
  n39 = Multiply(
   A=@n33.A
   B=MaterialFunctionCall!(Dynamic Param=@n30.Erosion, Alpha=@n31) {materialFunction=/Game/Effects/Fort_Effects/Materials/_Functions/Erode_Alpha}.Result
  )
  n40 = DepthFade(Opacity=@n39, FadeDistance=ScalarParameter() {defaultValue=20, parameterName=FadeDistance})
  n41 = ComponentMask(None=TextureCoordinate()) {g=true}
  n42 = ScalarParameter() {defaultValue=1.25, parameterName=Hide Distortion Edge on Top of Meshes}
  Multiply(
   A=Multiply(A=ShadingPathSwitch(Default=@n40, Mobile=@n39), B=@n37)
   B=LinearInterpolate(A=Power(Base=@n41, Exp=@n42) {constExponent=10}, Alpha=ScalarParameter() {parameterName=Is Radial Core}) {constAlpha=0}
  )
 ]
}
