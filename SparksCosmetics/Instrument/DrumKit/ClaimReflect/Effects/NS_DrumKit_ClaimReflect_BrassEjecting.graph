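# Niagara system that ejects spinning brass shell meshes for the ClaimReflect drum kit cosmetic, though both emitters are saved disabled.
# Shells spawn at RoundsPerSecond from skeletal mesh locations, driven by system-level velocity, gravity, size and rotation rate, then scale out.
# The GPU emitter collides when CloseToPlayer is true, damping spin after impact.
# The LowMed emitter is a shorter-lived variant without GPU simulation or collision.
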NiagaraSystem /SparksCosmetics/Instrument/DrumKit/ClaimReflect/Effects/NS_DrumKit_ClaimReflect_BrassEjecting
dependencies {
 usedDataInterfaces [
  &NiagaraDataInterfaceArrayFloat3
  /Script/Niagara.NiagaraDataInterfaceAsyncGpuTrace
  &NiagaraDataInterfaceCurve
  &NiagaraDataInterfaceSkeletalMesh
  &NiagaraDataInterfaceArrayPosition
 ]
 usedDynamicInputs []
 usedModules [
  &self:SystemSpawnScript.SystemScriptSource.SystemScriptGraph.NiagaraNodeAssignment_0.SetVariables_#1
  &self:SystemSpawnScript.SystemScriptSource.SystemScriptGraph.NiagaraNodeAssignment_2.SetVariables_#2
  &self:SystemSpawnScript.SystemScriptSource.SystemScriptGraph.NiagaraNodeAssignment_4.SetVariables_#3
  &SystemState
  &self:SystemSpawnScript.SystemScriptSource.SystemScriptGraph.NiagaraNodeAssignment_1.SetVariables_#4
  &self:ReadDeath
  &self:SystemSpawnScript.SystemScriptSource.SystemScriptGraph.NiagaraNodeAssignment_3.SetVariables_#5
 ]
}
emitters {
 Shell_Eject_Loop {
  bEnabled=false
  simTarget=GPUComputeSim
  emitterUpdateScript [
   &EmitterState()
   &SpawnRate(Spawn Probability=&ClampFloat(Float=$Emitter.Age), SpawnRate=$User.RoundsPerSecond, Use Spawn Probability=true)
   &SpawnBurst_Instantaneous(Spawn Count=1, Use Spawn Probability=false)
  ]
  particleSpawnScript [
   &InitializeParticle(
    Lifetime=&MakeCustomFloatFromBool(
     Bool=$System.CloseToPlayer
     True Float=&MaskFloatBySpawnGroup(Masked Float=0.75, Passthrough Float=&RandomRangeFloat(Minimum=1, Maximum=2), Spawn Group Mask=1)
     False Float=&MaskFloatBySpawnGroup(Masked Float=0.5, Passthrough Float=1, Spawn Group Mask=1)
    )
    Mesh Scale Mode=::NewEnumerator3
    Mesh Uniform Scale=0.5
    Position Offset=<0 0 -10>
    UsePositionOffset=false
    Write Material Random=true
   )
   &SkeletalMeshLocation(
    Bone / Socket Index=4
    Bone / Socket Sampling Mode=::NewEnumerator5
    MaskSpawnGroup=false
    Mesh Sampling Type=::NewEnumerator6
    Position Sampling=::NewEnumerator1
    Sampled Position Offset=&SelectVectorFromArray(Direct Array Index=0, Vector Selection Array=DataInterface() {propertyValues {FloatData [<0 0 -50>]}})
    Sampled Position Offset Coordinate Space:ENiagaraSkelMeshTransforms=::NewEnumerator1
    Skeletal Mesh=$System.Skeletal Mesh
   )
   &AddVelocity(
    Cone Angle=25
    Cone Axis=&Add_Vector(A=$System.Velocity, B=&RandomVector(Vector Scale=0.1))
    Module.UseConeFalloff=false
    Module.UseVelDistribution=false
    Override Seed=false
    Velocity Mode=::NewEnumerator2
    Velocity Speed=&Multiply_Float(
     A=&VectorLength(Vector=$System.Velocity)
     B=&UniformAOrBFloat(
      A=&RandomRangeFloat(Minimum=0.85, Maximum=1.25)
      B=&RandomRangeFloat(Minimum=0.25, Maximum=1.7)
      Distribution Weight=0.25
      Evaluation Type=::NewEnumerator0
      Recalculate Random Each Loop=true
      Override Seed=false
     )
    )
   )
   /SetVariables(Particles.VisibilityTag=&SpawnGroup()) {enabled=false, isSetParameters=true, module=SetVariables_#6}
   &SkeletalMeshLocation(
    Bone / Socket Index=4
    Bone / Socket Sampling Mode=::NewEnumerator5
    MaskSpawnGroup=false
    Mesh Sampling Type=::NewEnumerator6
    Position Sampling=::NewEnumerator1
    Sampled Position Offset=&SelectVectorFromArray(Direct Array Index=0, Vector Selection Array=DataInterface() {propertyValues {FloatData [<0 0 0>]}})
    Sampled Position Offset Coordinate Space:ENiagaraSkelMeshTransforms=::NewEnumerator1
    Skeletal Mesh=$System.Skeletal Mesh
   ) {module=SkeletalMeshLocation001}
   &InitialMeshOrientation(Mesh Axis To Orient=<0 -1 0>, Mesh Orientation Mode=::NewEnumerator2, Orientation Vector=<0 1 0>, Use Rotation Vector=false)
   /SetVariables(
    Particles.SendCollisionData:NiagaraBool=&SetBoolByFloatComparison(A=&RandomRangeFloat(Minimum=0, Maximum=1), B=0.45, Comparison Type=::NewEnumerator6)
    Particles.UniformMeshOrientation_Initial:Vector3f=&RandomRangeVector(Minimum=<0.2 0.2 0.2>, Maximum=<1 1 1>)
   ) {isSetParameters=true, module=SetVariables_#7}
  ]
  particleUpdateScript [
   &ParticleState()
   &AccelerationForce(Acceleration=$System.Gravity)
   &InheritVelocity(Module.Limit Speed=true, Velocity Limit=10, Velocity Scale=$System.InheritParentVelocity)
   &Collision(
    Age Colliding Particles=false
    Collision Enabled=$System.CloseToPlayer
    Friction=0.35
    Friction During a Bounce=0.25
    GPU Collision Type=::NewEnumerator2
    KillOnCollision=false
    Mesh Dimensions:Vector3f=<3 1 1>
    Module.Control Rest State Via Surface Normal=false
    Module.EnableMaxCollisionCount=false
    Module.Randomize Collision Normal Vector=false
    Radius Calculation Type=::NewEnumerator1
    Restitution=0.3
   )
   &SolveForcesAndVelocity(Clamp Velocity=false, Limit Acceleration=false)
   &ScaleMeshSize(Scale Factor=$System.Size)
   &ScaleMeshSize(Scale Factor=&VectorFromFloat(Value=&FloatFromCurve(
      FloatCurve=DataInterface() {propertyValues {Curve {keys [<at 0.7517 1 RCIM_Cubic>, {interpMode=RCIM_Cubic, tangentMode=RCTM_User, time=0.8642, value=1.3298}, <at 1 0 RCIM_Cubic>]}}}
      Scale Curve=0.5
     ))) {module=ScaleMeshSize001}
   /Niagara/Modules/Update/Orientation/V2/UpdateMeshOrientation(Rotation Vector:Vector3f=&Multiply_Vector(
     A=$Particles.UniformMeshOrientation_Initial
     B=&MakeCustomVectorFromBool(
      Bool=$Particles.HasCollided
      True Vector=&Multiply_VectorByFloat(
       Vector=$System.ShellRotationRate
       Float=&Lerp_Float(A=0.1, B=0, Alpha=&Power(a=$Particles.NormalizedAge, b=0.75))
      )
      False Vector=$System.ShellRotationRate
     )
    ))
   &DynamicMaterialParameters(
    Index 0 Param 1=&FloatFromCurve(FloatCurve=DataInterface() {propertyValues {Curve {keys [{arriveTangent=-3.3056, interpMode=RCIM_Cubic, leaveTangent=-3.3056, tangentMode=RCTM_User, time=0, value=2}, <at 1 1>]}}})
    Index 0 Param 2=0.25
    Index 0 Param 3=0.25
    Module.Param0WriteEnabled=true
    Module.Param1WriteEnabled=true
    Module.Param2WriteEnabled=true
    Module.Param3WriteEnabled=false
   )
   /SetVariables(
    Particles.CollisionHappen:NiagaraBool=$Transient.CollisionValid
    Particles.CollisionPos:Vector3f=$Transient.CollisionWorldPosition
    Particles.CollisionVector:Vector3f=$Particles.Collision.SafeCombinedCollisionNormal
   ) {isSetParameters=true}
  ]
  rendererClasses [&NiagaraMeshRendererProperties]
  renderers [{rendererClass=&NiagaraMeshRendererProperties, rendererIndex=0}]
 }
 Shell_Eject_Loop_LowMed {
  bEnabled=false
  emitterUpdateScript [
   &EmitterState()
   &SpawnRate(Spawn Probability=&ClampFloat(Float=$Emitter.Age), SpawnRate=$User.RoundsPerSecond, Use Spawn Probability=true)
   &SpawnBurst_Instantaneous(Spawn Count=1, Use Spawn Probability=false)
  ]
  particleSpawnScript [
   &InitializeParticle(
    Lifetime=&MakeCustomFloatFromBool(Bool=$System.CloseToPlayer, True Float=&RandomRangeFloat(Minimum=0.45, Maximum=0.5), False Float=0.45)
    Mesh Scale Mode=::NewEnumerator3
    Mesh Uniform Scale=0.5
    Position Offset=<0 0 -10>
    UsePositionOffset=false
    Write Material Random=true
   )
   &SkeletalMeshLocation(
    Bone / Socket Index=4
    Bone / Socket Sampling Mode=::NewEnumerator5
    MaskSpawnGroup=false
    Mesh Sampling Type=::NewEnumerator6
    Position Sampling=::NewEnumerator2
    Sampled Position Offset=&SelectVectorFromArray(Direct Array Index=0, Vector Selection Array=DataInterface() {propertyValues {FloatData [<0 0 0>]}})
    Sampled Position Offset Coordinate Space:ENiagaraSkelMeshTransforms=::NewEnumerator1
    Skeletal Mesh=$System.Skeletal Mesh
   )
   /SetVariables(Particles.VisibilityTag=&SpawnGroup()) {enabled=false, isSetParameters=true, module=SetVariables_#6}
   &CylinderLocation(Cylinder Height=0, Cylinder Radius=0, MaskSpawnGroup=false, Override Seed=false, UseOwnerScale=true)
   &SkeletalMeshLocation(
    Bone / Socket Index=4
    Bone / Socket Sampling Mode=::NewEnumerator5
    MaskSpawnGroup=false
    Mesh Sampling Type=::NewEnumerator6
    Position Sampling=::NewEnumerator1
    Sampled Position Offset=&SelectVectorFromArray(Direct Array Index=0, Vector Selection Array=DataInterface() {propertyValues {FloatData [<0 0 -10>]}})
    Sampled Position Offset Coordinate Space:ENiagaraSkelMeshTransforms=::NewEnumerator1
    Skeletal Mesh=$System.Skeletal Mesh
   ) {module=SkeletalMeshLocation001}
   &AddVelocity(Velocity=&Multiply_VectorByFloat(
     Vector=$System.Velocity
     Float=&UniformAOrBFloat(
      A=&RandomRangeFloat(Minimum=0.85, Maximum=1.25)
      B=&RandomRangeFloat(Minimum=0.25, Maximum=1.7)
      Distribution Weight=0.25
      Evaluation Type=::NewEnumerator0
      Recalculate Random Each Loop=true
      Override Seed=false
     )
    ))
   &InitialMeshOrientation(Mesh Axis To Orient=<0 -1 0>, Mesh Orientation Mode=::NewEnumerator2, Use Rotation Vector=false)
   /SetVariables(Particles.UniformMeshOrientation_Initial:Vector3f=&RandomRangeVector(Minimum=<0.2 0.2 0.2>, Maximum=<1 1 1>)) {isSetParameters=true, module=SetVariables_#7}
  ]
  particleUpdateScript [
   &ParticleState()
   &AccelerationForce(Acceleration=$System.Gravity)
   &InheritVelocity(Module.Limit Speed=true, Velocity Limit=10, Velocity Scale=$System.InheritParentVelocity)
   &SolveForcesAndVelocity(Clamp Velocity=false, Limit Acceleration=false)
   &ScaleMeshSize(Scale Factor=$System.Size)
   &ScaleMeshSize(Scale Factor=&VectorFromFloat(Value=&FloatFromCurve(
      FloatCurve=DataInterface() {propertyValues {Curve {keys [<at 0.7517 1 RCIM_Cubic>, {interpMode=RCIM_Cubic, tangentMode=RCTM_User, time=0.8642, value=1.3298}, <at 1 0 RCIM_Cubic>]}}}
      Scale Curve=0.7
     ))) {module=ScaleMeshSize001}
   /Niagara/Modules/Update/Orientation/V3/UpdateMeshOrientation(Rotation Vector=&Multiply_Vector(
     A=$Particles.UniformMeshOrientation_Initial
     B=&MakeCustomVectorFromBool(
      Bool=false
      True Vector=&Multiply_VectorByFloat(
       Vector=$System.ShellRotationRate
       Float=&Lerp_Float(A=0.1, B=0, Alpha=&Power(a=$Particles.NormalizedAge, b=0.75))
      )
      False Vector=$System.ShellRotationRate
     )
    )) {module=UpdateMeshOrientation001}
   &DynamicMaterialParameters(
    Index 0 Param 1=&FloatFromCurve(FloatCurve=DataInterface() {propertyValues {Curve {keys [{arriveTangent=-3.3056, interpMode=RCIM_Cubic, leaveTangent=-3.3056, tangentMode=RCTM_User, time=0, value=2}, <at 1 1>]}}})
    Index 0 Param 2=0.25
    Index 0 Param 3=0.25
    Module.Index0Float0Write=true
    Module.Index0Float1Write=true
    Module.Index0Float2Write=true
    Module.Index0Float3Write=false
   ) {module=DynamicMaterialParameters001}
  ]
  rendererClasses [&NiagaraMeshRendererProperties]
  renderers [{rendererClass=&NiagaraMeshRendererProperties, rendererIndex=0}]
 }
}
summary {
 systemName=NS_DrumKit_ClaimReflect_BrassEjecting
 emitters [
  {bEnabled=false, emitterName=Shell_Eject_Loop, simTarget=GPUComputeSim, rendererClasses [&NiagaraMeshRendererProperties]}
  {bEnabled=false, emitterName=Shell_Eject_Loop_LowMed, rendererClasses [&NiagaraMeshRendererProperties]}
 ]
 userVariables [
  User.bIsFrontEndPreview {type=NiagaraBool}
  User.FiringRate {type=NiagaraFloat, defaultValue=0.1, description=""}
  User.RoundsPerSecond {type=NiagaraFloat, defaultValue=35, description=""}
 ]
}
