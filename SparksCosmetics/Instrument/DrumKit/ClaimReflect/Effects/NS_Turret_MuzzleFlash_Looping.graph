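# Looping turret muzzle flash Niagara system for a Sparks DrumKit cosmetic effect.
# GPU Muzzle_Core_Elements draws short-lived mesh and sprite flashes at User.RoundsPerSecond with randomized spawn probability.
# Ejected_Material layers burst sprites through spawn-group masks, using User smoke color and emissive inputs, with one group gated by ADS and heated-element flags.
# Lighting adds a brief warm fading light through light and component renderers, skipped when the muzzle is suppressed.
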NiagaraSystem /SparksCosmetics/Instrument/DrumKit/ClaimReflect/Effects/NS_Turret_MuzzleFlash_Looping
dependencies {
 usedDataInterfaces [&NiagaraDataInterfaceArrayFloat3, &NiagaraDataInterfaceCurve]
 usedDynamicInputs []
 usedModules [
  &self:SystemRandomScale
  &SystemState
 ]
}
emitters {
 Ejected_Material {
  emitterUpdateScript [
   &EmitterState()
   &SpawnBurst_Instantaneous(Spawn Count=3, Use Spawn Probability=false, UseLoopCountLimit=false)
   &SpawnBurst_Instantaneous(
    Spawn Count=&RandomRangeInt(Minimum=0, Maximum=1, Evaluation Type=::NewEnumerator1)
    Spawn Group=1
    Use Spawn Probability=false
    UseLoopCountLimit=false
   ) {module=SpawnBurst_Instantaneous001}
   &SpawnBurst_Instantaneous(
    Spawn Count=&MakeCustomIntFromBool(
     Bool=$System.isADS
     True INT=0
     False INT=&MakeCustomIntFromBool(Bool=$System.bUseHeatedElements, True INT=1, False INT=0)
    )
    Spawn Group=2
    Use Spawn Probability=false
    UseLoopCountLimit=false
   ) {module=SpawnBurst_Instantaneous002}
  ]
  particleSpawnScript [
   /SetVariables(Particles.VisibilityTag=&SpawnGroup()) {isSetParameters=true}
   &InitializeParticle(
    Color=$User.SmokeInitialColor
    Lifetime=&MaskFloatBySpawnGroup(
     Masked Float=&RandomRangeFloat(Minimum=0.15, Maximum=0.45)
     Passthrough Float=&MaskFloatBySpawnGroup(
      Masked Float=&RandomRangeFloat(Minimum=0.055, Maximum=0.15)
      Passthrough Float=&MaskFloatBySpawnGroup(Masked Float=&RandomRangeFloat(Minimum=1, Maximum=1.1), Passthrough Float=0.15, Spawn Group Mask=2)
      Spawn Group Mask=1
     )
    )
    Position Offset=&MakeCustomVectorFromBool(
     Bool=&SetBoolByIntComparison(A=&SpawnGroup(), B=2, Comparison Type=::NewEnumerator2)
     True Vector=<0 0 0>
     False Vector=&Multiply_VectorByFloat(
      Vector=<15 0 0>
      Float=&ClampFloat(Float=&Add_Float(
        A=&ReturnNormalizedExecIndex()
        B=&MakeCustomFloatFromBool(Bool=$System.isADS, True Float=0.0063, False Float=0.05)
       ))
     )
    )
    Position Offset Coordinate Space=::Local
    Sprite Rotation Angle=&MaskFloatBySpawnGroup(
     Masked Float=&RandomRangeFloat(Minimum=0, Maximum=360)
     Passthrough Float=&MaskFloatBySpawnGroup(
      Masked Float=0
      Passthrough Float=&MaskFloatBySpawnGroup(
       Masked Float=&Add_Float(A=30, B=&RandomRangeFloat(Minimum=-15, Maximum=15))
       Passthrough Float=0
       Spawn Group Mask=2
      )
      Spawn Group Mask=1
     )
    )
    Sprite Rotation Mode=::NewEnumerator0
    Sprite Size Mode=::NewEnumerator3
    Uniform Sprite Size=&MaskFloatBySpawnGroup(
     Masked Float=150
     Passthrough Float=&MaskFloatBySpawnGroup(
      Masked Float=&RandomRangeFloat(Minimum=0.75, Maximum=1.5)
      Passthrough Float=&RandomRangeFloat(Minimum=30, Maximum=50)
      Spawn Group Mask=1
     )
    )
    UsePositionOffset=true
    Write Material Random=true
   )
   &ShapeLocation(
    Hemisphere Distribution=<1 0.5>
    MaskSpawnGroup=true
    Offset=<20 0 0>
    Offset Coordinate Space=::Local
    Offset Mode=::NewEnumerator0
    Override Seed=false
    Rotation Mode=::NewEnumerator1
    Spawn Group Mask=1
    Sphere Radius=1
    UseOwnerScale=true
   )
   &AddVelocity(
    Cone Angle=&MaskFloatBySpawnGroup(Masked Float=10, Passthrough Float=25)
    Cone Axis=<0 0 0>
    Module.UseConeFalloff=false
    Module.UseVelDistribution=false
    Override Seed=false
    Velocity Mode=::NewEnumerator2
    Velocity Speed=&MaskFloatBySpawnGroup(
     Masked Float=&RandomRangeFloat(Minimum=1100, Maximum=1500)
     Passthrough Float=&MaskFloatBySpawnGroup(Masked Float=&RandomRangeFloat(Minimum=900, Maximum=1200), Passthrough Float=0, Spawn Group Mask=1)
    )
   )
   &CameraOffset(Camera Offset Amount=&MaskFloatBySpawnGroup(Masked Float=10, Passthrough Float=0))
   /SetVariables(Particles.Position=&AddVectorToPosition(
     Position=$Particles.Position
     Vector=&TransformVector(
      Vector=&SelectVectorFromArray(
       Direct Array Index=&ReturnParticleIDAttributes()
       Direct Array Mode=::NewEnumerator1
       Vector Selection Array=DataInterface() {propertyValues {FloatData [<0 0 5>, <0 -5 -4.5>, <0 5 -4.5>]}}
      )
      Destination Space=::Simulation
     )
    )) {enabled=false, isSetParameters=true}
  ]
  particleUpdateScript [
   &ParticleState()
   &AccelerationForce(Acceleration=&MaskVectorBySpawnGroup(
     Masked Vector=<0 0 150>
     Passthrough Vector=&MaskVectorBySpawnGroup(Masked Vector=<0 0 -500>, Passthrough Vector=<0 0 150>, Spawn Group Mask=1)
    ))
   &VectorNoiseForce(
    Force Amount=&MaskFloatBySpawnGroup(Masked Float=5000, Passthrough Float=0, Spawn Group Mask=1)
    Module.Use Random Noise Time Interval=true
    Random Noise Time Interval=0.05
   )
   &Drag(
    Drag=&MaskFloatBySpawnGroup(
     Masked Float=&Multiply_Float(
      A=&RandomRangeFloat(Minimum=8, Maximum=13)
      B=&FloatFromCurve(FloatCurve=DataInterface() {propertyValues {Curve {keys [{arriveTangent=-5.0888, interpMode=RCIM_Cubic, leaveTangent=-5.0888, tangentMode=RCTM_User, time=0.5, value=1}, <at 1 0 RCIM_Cubic>]}}})
     )
     Passthrough Float=&MaskFloatBySpawnGroup(Masked Float=&RandomRangeFloat(Minimum=1, Maximum=5), Passthrough Float=3, Spawn Group Mask=1)
    )
    Use Linear Drag=true
    Use Rotational Drag=false
   )
   &SolveForcesAndVelocity(Clamp Velocity=false, Limit Acceleration=false)
   &SpriteRotationRate(Rotation Rate=&MaskFloatBySpawnGroup(
     Masked Float=&FloatFromCurve(
      FloatCurve=DataInterface() {propertyValues {Curve {keys [<at -0.0046 1.0113 RCIM_Cubic>,<at 1.0011 0.09 RCIM_Cubic>]}}}
      Scale Curve=&RandomRangeFloat(Minimum=-3, Maximum=3)
     )
     Passthrough Float=0
    ))
   &ScaleSpriteSizeBySpeed(
    Max Scale Factor=&MaskVector2dBySpawnGroup(
     Masked Vector2D=&RandomRangeVector2D(Minimum=<1 5>, Maximum=<1 50>)
     Passthrough Vector2D=<1 1>
     Spawn Group Mask=1
    )
    Min Scale Factor=&MaskVector2dBySpawnGroup(
     Masked Vector2D=&RandomRangeVector2D(Minimum=<0.5 0.5>, Maximum=<1 1>)
     Passthrough Vector2D=<1 1>
     Spawn Group Mask=1
    )
   )
   &ScaleSpriteSize(
    Scale Sprite Size Mode=::NewEnumerator0
    Uniform Scale Factor=&MaskFloatBySpawnGroup(
     Masked Float=&FloatFromCurve(
      FloatCurve=DataInterface() {propertyValues {Curve {keys [
          {arriveTangent=1.4341, interpMode=RCIM_Cubic, leaveTangent=1.4341, leaveTangentWeight=0.8692, tangentMode=RCTM_User, time=0, value=0.25}
          {arriveTangentWeight=0.6592, interpMode=RCIM_Cubic, tangentMode=RCTM_User, time=1, value=1}
         ]}}}
      Scale Curve=0.66
     )
     Passthrough Float=&MaskFloatBySpawnGroup(
      Masked Float=&FloatFromCurve(
       FloatCurve=DataInterface() {propertyValues {Curve {keys [{tangentMode=RCTM_User, time=0.8, value=1}, {tangentMode=RCTM_User, time=1, value=0}]}}}
       Scale Curve=&RandomRangeFloat(Minimum=0.1, Maximum=1)
      )
      Passthrough Float=&FloatFromCurve(FloatCurve=DataInterface() {propertyValues {Curve {keys [
           {arriveTangent=0.5942, interpMode=RCIM_Cubic, leaveTangent=0.5942, tangentMode=RCTM_User, time=0, value=0.33}
           {interpMode=RCIM_Cubic, leaveTangent=0.6906, time=1, value=1}
          ]}}})
      Spawn Group Mask=1
     )
    )
   )
   &DynamicMaterialParameters(
    Index 0 Param 1=&Add_Float(
     A=&RandomRangeFloat(Minimum=0, Maximum=1)
     B=&FloatFromCurve(FloatCurve=DataInterface() {propertyValues {Curve {keys [{arriveTangent=-1.1823, interpMode=RCIM_Cubic, leaveTangent=-1.1823, tangentMode=RCTM_User, time=0, value=1}, <at 1 0 RCIM_Cubic>]}}})
    )
    Index 0 Param 2=&FloatFromCurve(FloatCurve=DataInterface() {propertyValues {Curve {keys [
         {arriveTangent=0.8564, interpMode=RCIM_Cubic, leaveTangent=0.8564, tangentMode=RCTM_User, time=0.0046, value=-0.9753}
         {arriveTangent=4.2876, interpMode=RCIM_Cubic, leaveTangent=5.5268, tangentMode=RCTM_User, time=1, value=1}
        ]}}})
    Index 0 Param 3=&FloatFromCurve(
     FloatCurve=DataInterface() {propertyValues {Curve {keys [
         {arriveTangent=-15.2571, interpMode=RCIM_Cubic, leaveTangent=-15.2571, tangentMode=RCTM_User, time=0, value=1}
         {tangentMode=RCTM_User, time=0.2, value=0}
        ]}}}
     Scale Curve=$User.SmokeEmissiveAmount
    )
    Index 0 Param 4=&FloatFromCurve(FloatCurve=DataInterface() {propertyValues {Curve {keys [{arriveTangent=-3.2497, interpMode=RCIM_Cubic, leaveTangent=-3.2497, tangentMode=RCTM_User, time=0, value=1}, <at 1 0 RCIM_Cubic>]}}})
    Index 1 Param 1=&FloatFromCurve(FloatCurve=DataInterface() {propertyValues {Curve {keys [{arriveTangent=1.674, interpMode=RCIM_Cubic, leaveTangent=1.674, tangentMode=RCTM_User, time=0, value=0.5}, <at 1 1 RCIM_Cubic>]}}})
    Index 1 Param 2=&RandomRangeFloat(Minimum=0, Maximum=1)
    Index 1 Param 3=&FloatFromCurve(FloatCurve=DataInterface() {propertyValues {Curve {keys [<at 0 10 RCIM_Cubic>, {arriveTangent=-85.6983, interpMode=RCIM_Cubic, leaveTangent=-85.6983, tangentMode=RCTM_User, time=0.25, value=1}]}}})
    Index 1 Param 4=&RandomRangeFloat(Minimum=2, Maximum=6)
    Index 2 Param 1=&FloatFromCurve(FloatCurve=DataInterface() {propertyValues {Curve {keys [
         {interpMode=RCIM_Cubic, leaveTangent=1.673, tangentMode=RCTM_Break, time=0, value=0.5}
         {interpMode=RCIM_Cubic, tangentMode=RCTM_Break, time=1, value=1}
        ]}}})
    Index 2 Param 2=&FloatFromCurve(
     FloatCurve=DataInterface() {propertyValues {Curve {keys [
         {interpMode=RCIM_Cubic, leaveTangentWeight=0.397, tangentMode=RCTM_Break, time=0, value=0.7}
         {arriveTangentWeight=0.4342, interpMode=RCIM_Cubic, tangentMode=RCTM_Break, time=1, value=1}
        ]}}}
     Scale Curve=&RandomRangeFloat(Minimum=0, Maximum=1)
    )
    Index 2 Param 3=&FloatFromCurve(
     FloatCurve=DataInterface() {propertyValues {Curve {keys [
         {interpMode=RCIM_Cubic, leaveTangent=2.6489, leaveTangentWeight=0.3345, tangentMode=RCTM_Break, time=0, value=0}
         {arriveTangentWeight=0.3308, interpMode=RCIM_Cubic, tangentMode=RCTM_Break, time=1, value=1}
        ]}}}
     Scale Curve=5
    )
    Index 2 Param 4=&FloatFromCurve(FloatCurve=DataInterface() {propertyValues {Curve {keys [
         {interpMode=RCIM_Cubic, leaveTangent=0.5158, leaveTangentWeight=0.3333, tangentMode=RCTM_Break, time=0, value=0.8}
         {arriveTangentWeight=0.3592, interpMode=RCIM_Cubic, tangentMode=RCTM_Break, time=1, value=1}
        ]}}})
    Module.Index0Float0Write=true
    Module.Index0Float1Write=true
    Module.Index0Float2Write=true
    Module.Index0Float3Write=true
    Module.Index1Float0Write=true
    Module.Index1Float1Write=true
    Module.Index1Float2Write=true
    Module.Index1Float3Write=true
    Module.Index2Float0Write=true
    Module.Index2Float1Write=true
    Module.Index2Float2Write=true
    Module.Index2Float3Write=true
    Module.Index3Float0Write=true
    Module.Index3Float1Write=true
    Module.Index3Float2Write=true
    Module.Index3Float3Write=true
    Write Parameter Index 1=::NewEnumerator1
    Write Parameter Index 2=::NewEnumerator1
    Write Parameter Index 3=::NewEnumerator1
   )
   &ScaleColor(
    Scale Alpha=&MaskFloatBySpawnGroup(
     Masked Float=&FloatFromCurve(FloatCurve=DataInterface() {propertyValues {Curve {keys [
          {arriveTangent=9.0725, interpMode=RCIM_Cubic, leaveTangent=9.0725, tangentMode=RCTM_User, time=0, value=0}
          {arriveTangentWeight=0.1274, interpMode=RCIM_Cubic, leaveTangentWeight=0.7778, tangentMode=RCTM_User, time=0.1408, value=0.9847}
          {arriveTangent=-0.0405, interpMode=RCIM_Cubic, leaveTangent=-0.0405, tangentMode=RCTM_User, time=1, value=0}
         ]}}})
     Passthrough Float=&MaskFloatBySpawnGroup(
      Masked Float=&FloatFromCurve(FloatCurve=DataInterface() {propertyValues {Curve {keys [<at 0 3 RCIM_Cubic>,<at 1 1 RCIM_Cubic>]}}})
      Passthrough Float=&FloatFromCurve(FloatCurve=DataInterface() {propertyValues {Curve {keys [
           {interpMode=RCIM_Cubic, tangentMode=RCTM_User, time=0, value=0}
           {interpMode=RCIM_Cubic, tangentMode=RCTM_User, time=0.2, value=0.9847}
           {interpMode=RCIM_Cubic, tangentMode=RCTM_User, time=0.66, value=1}
           {arriveTangent=-0.0405, interpMode=RCIM_Cubic, leaveTangent=-0.0405, tangentMode=RCTM_User, time=1, value=0}
          ]}}})
      Spawn Group Mask=1
     )
    )
    Scale RGB=&MaskVectorBySpawnGroup(
     Masked Vector=&Multiply_VectorByFloat(
      Vector=<1 1 1>
      Float=&FloatFromCurve(FloatCurve=DataInterface() {propertyValues {Curve {keys [<at 0 1>,<at 1 0>]}}})
     )
     Passthrough Vector=<1 1 1>
    )
    ScaleA=true
    ScaleRGB=true
   )
   /SetVariables(Particles.Sparks Dynamic Parameter:Vector4f=&MakeVector4(X=2, Y=20, Z=90, W=1)) {isSetParameters=true}
   /SetVariables(
    Particles.HeatedSmoke_DynamicParam_00:Vector4f=&MakeVector4(X=&RandomRangeFloat(Minimum=0, Maximum=32), Y=0, Z=0, W=0)
    Particles.HeatedSmoke_DynamicParam_02:Vector4f=&MakeVector4(
     X=&FloatFromCurve(FloatCurve=DataInterface() {propertyValues {Curve {keys [
          {arriveTangent=5.827, interpMode=RCIM_Cubic, leaveTangent=5.827, tangentMode=RCTM_User, time=0, value=0}
          {interpMode=RCIM_Cubic, tangentMode=RCTM_User, time=0.25, value=1}
         ]}}})
     Y=&Add_Float(
      A=&RandomRangeFloat(Minimum=-1, Maximum=1)
      B=&FloatFromCurve(
       FloatCurve=DataInterface() {propertyValues {Curve {keys [{tangentMode=RCTM_User, time=0, value=1}, {tangentMode=RCTM_User, time=1, value=0}]}}}
       Scale Curve=&RandomRangeFloat(Minimum=0.25, Maximum=0.5)
      )
     )
     Z=&FloatFromCurve(
      FloatCurve=DataInterface() {propertyValues {Curve {keys [
          {arriveTangent=0.103, interpMode=RCIM_Cubic, leaveTangent=0.103, tangentMode=RCTM_User, time=0, value=0.25}
          {arriveTangent=0.5215, arriveTangentWeight=0.837, interpMode=RCIM_Cubic, leaveTangent=0.5215, tangentMode=RCTM_User, time=1, value=1}
         ]}}}
      Scale Curve=3
     )
     W=&RandomRangeFloat(Minimum=0.75, Maximum=1)
    )
   ) {isSetParameters=true}
  ]
  rendererClasses [&NiagaraSpriteRendererProperties]
  renderers [
   {rendererClass=&NiagaraSpriteRendererProperties}
   {rendererClass=&NiagaraSpriteRendererProperties, rendererIndex=1}
   {rendererClass=&NiagaraSpriteRendererProperties, rendererIndex=2}
   {rendererClass=&NiagaraSpriteRendererProperties, rendererIndex=3}
  ]
 }
 Lighting {
  emitterUpdateScript [
   &EmitterState()
   &SpawnBurst_Instantaneous(
    Spawn Count=/Niagara/DynamicInputs/Math/Min_Int(
     A:NiagaraInt32=&MakeCustomIntFromBool(Bool=$System.isMuzzleSupressor, True INT=0, False INT=1)
     B:NiagaraInt32=2
    )
    Use Spawn Probability=false
    UseLoopCountLimit=false
   )
  ]
  particleSpawnScript [&InitializeParticle(
    Lifetime=0.1
    Position Offset=<7 0 0>
    Sprite Size Mode=::NewEnumerator3
    Uniform Sprite Size=15
    UsePositionOffset=true
    Write Material Random=true
   )]
  particleUpdateScript [
   &ParticleState()
   &ScaleColor(
    Scale Alpha=&Lerp_Float(
     A=20
     B=0
     Alpha=&FloatFromCurve(FloatCurve=DataInterface() {propertyValues {Curve {keys [<at 0 1>,<at 1 0.3>]}}})
    )
    Scale RGB=&MakeVectorFromLinearColorRGB(Color=<rgba 1 0.1595 0.0143 1>)
    ScaleA=true
    ScaleRGB=true
   )
   /SetVariables(
    Particles.LightExponent=5
    Particles.LightRadius=200
    Particles.PointLightIntensity:NiagaraFloat=&Multiply_Float(
     A=2500
     B=&FloatFromCurve(FloatCurve=DataInterface() {propertyValues {Curve {keys [<at 0 1>,<at 1 0>]}}})
    )
   ) {isSetParameters=true}
  ]
  rendererClasses [&NiagaraLightRendererProperties, &NiagaraComponentRendererProperties]
  renderers [
   {rendererClass=&NiagaraLightRendererProperties, rendererIndex=0}
   {rendererClass=&NiagaraComponentRendererProperties, rendererIndex=1}
  ]
 }
 Muzzle_Core_Elements {
  simTarget=GPUComputeSim
  emitterUpdateScript [
   &EmitterState()
   &SpawnBurst_Instantaneous(Use Spawn Probability=false, UseLoopCountLimit=false)
   &SpawnRate(Spawn Probability=0.85, SpawnRate=$User.RoundsPerSecond, Use Spawn Probability=true)
  ]
  particleSpawnScript [
   /SetVariables(Particles.VisibilityTag=&SpawnGroup()) {isSetParameters=true}
   &InitializeParticle(
    Color Minimum=<rgba 0.8854 0.6917 0.697 0.5>
    Color Mode=::NewEnumerator2
    Lifetime=&RandomRangeFloat(Minimum=0.045, Maximum=0.075)
    Mesh Scale=<2 0.2 0.2>
    Mesh Scale Mode=::NewEnumerator4
    Sprite Rotation Mode=::NewEnumerator2
    Sprite Size=&Vector2DFromFloat(Value=&RandomRangeFloat(Minimum=45, Maximum=80))
    Sprite Size Mode=::NewEnumerator4
    Sprite UV Mode=::NewEnumerator3
    UsePositionOffset=true
    Write Material Random=true
   )
   &InitialMeshOrientation(
    Mesh Orientation Mode=::NewEnumerator2
    Orientation Vector=&MaskVectorBySpawnGroup(
     Masked Vector=<1 0 0>
     Passthrough Vector=&self:Rotation(
      Default:Vector3f=<0 1 0>
      float:NiagaraFloat=&Divide_Float(A=&ReturnNormalizedExecIndex(Normalized Index Includes One=false), B=1)
     )
    )
    Rotation=&RandomRangeVector(Minimum=<0 0 0>, Maximum=<1 0 0>)
    Use Rotation Vector=true
   )
   &CameraOffset(Camera Offset Amount=1)
   &SpriteFacingAndAlignment(
    Sprite Alignment=&MaskVectorBySpawnGroup(Masked Vector=<0 1 0>, Passthrough Vector=<1 0 0>)
    Sprite Facing=&MaskVectorBySpawnGroup(
     Masked Vector=<1 0 0>
     Passthrough Vector=&Lerp_Vector(A=<0 1 0>, B=<0 -1 0>, Alpha=&MakeFloatFromInt(Int=$Particles.PartitionParticles.Partition))
    )
    WriteAlignment=true
    WriteFacing=true
   )
   /SetVariables(
    Particles.Scale=&Multiply_Vector(A=$Particles.Scale, B=&RandomRangeVector(Minimum=<0.25 1 1>, Maximum=<1.5 1.5 1.5>))
    Particles.SpriteCameraOffset=5
   ) {isSetParameters=true}
   /SetVariables(
    Particles.ColorGlow:LinearColor=&Multiply_LinearColorByFloat(Linear Color=$Particles.Color, Float=&RandomRangeFloat(Minimum=0.4, Maximum=0.45))
    Particles.MeshSizeGlow:Vector3f=&Multiply_Vector(A=$Particles.Scale, B=&RandomRangeVector(Minimum=<0.75 4 4>, Maximum=<1 7.5 7.5>))
   ) {isSetParameters=true}
  ]
  particleUpdateScript [
   &ParticleState()
   &ScaleMeshSize(Scale Factor=&VectorFromFloat(Value=&FloatFromCurve(FloatCurve=DataInterface() {propertyValues {Curve {keys [<at 0 0.5>,<at 1 1>]}}})))
   &DynamicMaterialParameters(
    Index 0 Param 1=&FloatFromCurve(
     FloatCurve=DataInterface() {propertyValues {Curve {keys [
         {arriveTangent=2.9073, interpMode=RCIM_Cubic, leaveTangent=2.9073, leaveTangentWeight=0.3077, tangentMode=RCTM_User, time=0.35, value=0}
         {arriveTangentWeight=0.8354, interpMode=RCIM_Cubic, tangentMode=RCTM_User, time=1, value=1}
        ]}}}
     Scale Curve=&RandomRangeFloat(Minimum=0.5, Maximum=1)
    )
    Index 0 Param 2=&FloatFromCurve(FloatCurve=DataInterface() {propertyValues {Curve {keys [
         {interpMode=RCIM_Cubic, leaveTangentWeight=0.3627, tangentMode=RCTM_User, time=0.4, value=1}
         {arriveTangent=-1.799, arriveTangentWeight=1.0345, interpMode=RCIM_Cubic, leaveTangent=-1.799, tangentMode=RCTM_User, time=1, value=0}
        ]}}})
    Index 0 Param 3=&MakeCustomFloatFromBool(Bool=&RandomBool(), True Float=1, False Float=0)
    Module.Index0Param0Write=true
    Module.Index0Param1Write=true
    Module.Index0Param2Write=true
    Module.Index0Param3Write=false
   ) {enabled=false}
   &ScaleSpriteSize(
    Scale Factor=&Vector2DFromFloat(Value=&FloatFromCurve(
      FloatCurve=DataInterface() {propertyValues {Curve {keys [
          {arriveTangent=1.3058, interpMode=RCIM_Cubic, leaveTangent=1.3058, tangentMode=RCTM_User, time=0, value=0.25}
          {arriveTangent=0.3286, tangentMode=RCTM_User, time=1, value=1}
         ]}}}
      Scale Curve=&RandomRangeFloat(Minimum=0.75, Maximum=1)
     ))
    Scale Sprite Size Mode=::NewEnumerator2
   )
  ]
  rendererClasses [&NiagaraMeshRendererProperties, &NiagaraSpriteRendererProperties]
  renderers [
   {rendererClass=&NiagaraMeshRendererProperties, rendererIndex=0}
   {rendererClass=&NiagaraSpriteRendererProperties, rendererIndex=1}
   {rendererClass=&NiagaraMeshRendererProperties, rendererIndex=2}
  ]
 }
}
summary {
 systemName=NS_Turret_MuzzleFlash_Looping
 emitters [
  {emitterName=Ejected_Material, rendererClasses [&NiagaraSpriteRendererProperties]}
  {
   emitterName=Muzzle_Core_Elements
   simTarget=GPUComputeSim
   rendererClasses [&NiagaraMeshRendererProperties, &NiagaraSpriteRendererProperties]
  }
  {emitterName=Lighting, rendererClasses [&NiagaraLightRendererProperties, &NiagaraComponentRendererProperties]}
 ]
 userVariables [
  User.FiringRate {type=NiagaraFloat, defaultValue=0.1, description=""}
  User.RoundsPerSecond {type=NiagaraFloat, defaultValue=35, description=""}
  User.SmokeEmissiveAmount {type=NiagaraFloat, defaultValue=1, description=""}
  User.SmokeInitialColor {type=LinearColor, description="", defaultValue {struct=LinearColor, value=<rgba 0.7326 0.7047 0.7047 0.75>}}
 ]
}
